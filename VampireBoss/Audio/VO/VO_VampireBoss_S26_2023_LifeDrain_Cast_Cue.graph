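# Sound cue for the Vampire Boss's voice line when it casts its Life Drain ability.
# A random node picks evenly among two LifeDrain start lines and two reused Alerted lines, then a modulator passes the result out.
# It plays in the critical enemy VO sound class with boosted volume, boss VO attenuation and a chatter concurrency set that governs overlapping lines.
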SoundCue /VampireBoss/Audio/VO/VO_VampireBoss_S26_2023_LifeDrain_Cast_Cue
root=Modulator_2
nodes [
 {id=Modulator_2, children [Random_0]}
 {
  id=Random_0
  children [WavePlayer_3,WavePlayer_5,WavePlayer_0,WavePlayer_1]
  properties {bShouldExcludeFromBranchCulling=true, weights [1,1,1,1]}
 }
 {id=WavePlayer_0, properties {soundWaveAssetPtr=/VampireBoss/Audio/VO/Waves/VampireBoss_S26_2023_Alerted_030}}
 {id=WavePlayer_1, properties {soundWaveAssetPtr=/VampireBoss/Audio/VO/Waves/VampireBoss_S26_2023_Alerted_031}}
 {id=WavePlayer_3, properties {soundWaveAssetPtr=/VampireBoss/Audio/VO/Waves/VampireBoss_S26_2023_LifeDrain_Start_020}}
 {id=WavePlayer_5, properties {soundWaveAssetPtr=/VampireBoss/Audio/VO/Waves/VampireBoss_S26_2023_LifeDrain_Start_021}}
]
properties {
 attenuationSettings=/VampireBoss/Audio/NPC_VampireBoss_VO_ST_NoAA_Att
 maxDistance=3200
 soundClassObject=&Fort_VO_Enemy_Critical
 volumeMultiplier=2.7
 concurrencySet [&/VampireBoss_Chatter_Conc]
}
