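# Sound cue for the Vampire Boss's voice line at the end of its Life Drain.
# A Random node picks one of four equally weighted voice waves and is excluded from branch culling.
# The cue plays boosted on the critical enemy VO sound class, using the boss VO attenuation and a chatter concurrency group.
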SoundCue /VampireBoss/Audio/VO/VO_VampireBoss_S26_2023_LifeDrain_End_Cue
root=Random_0
nodes [
 {
  id=Random_0
  children [WavePlayer_5,WavePlayer_6,WavePlayer_7,WavePlayer_8]
  properties {bShouldExcludeFromBranchCulling=true, weights [1,1,1,1]}
 }
 {id=WavePlayer_5, properties {soundWaveAssetPtr=/VampireBoss/Audio/VO/Waves/VampireBoss_S26_2023_LifeDrain_End_010}}
 {id=WavePlayer_6, properties {soundWaveAssetPtr=/VampireBoss/Audio/VO/Waves/VampireBoss_S26_2023_LifeDrain_End_011}}
 {id=WavePlayer_7, properties {soundWaveAssetPtr=/VampireBoss/Audio/VO/Waves/VampireBoss_S26_2023_LifeDrain_End_022}}
 {id=WavePlayer_8, properties {soundWaveAssetPtr=/VampireBoss/Audio/VO/Waves/VampireBoss_S26_2023_LifeDrain_End_030}}
]
properties {
 attenuationSettings=/VampireBoss/Audio/NPC_VampireBoss_VO_ST_Att
 maxDistance=3200
 soundClassObject=&Fort_VO_Enemy_Critical
 volumeMultiplier=2.7
 concurrencySet [&/VampireBoss_Chatter_Conc]
}
