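# Sound Cue for the Vampire Boss's idle voice lines.
# A Random node, excluded from branch culling, picks one of five equally weighted wave players, each holding a different idle recording.
# Playback uses the shared close-range NPC voice attenuation, the critical enemy VO sound class and a slight volume boost, with a Vampire Boss chatter concurrency set governing overlapping voices.
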SoundCue /VampireBoss/Audio/VO/VampireBoss_S26_2023_Idle_Cue
root=Random_0
nodes [
 {
  id=Random_0
  children [WavePlayer_0,WavePlayer_1,WavePlayer_2,WavePlayer_3,WavePlayer_4]
  properties {bShouldExcludeFromBranchCulling=true, weights [1,1,1,1,1]}
 }
 {id=WavePlayer_0, properties {soundWaveAssetPtr=/VampireBoss/Audio/VO/Waves/VampireBoss_S26_2023_Idle_010}}
 {id=WavePlayer_1, properties {soundWaveAssetPtr=/VampireBoss/Audio/VO/Waves/VampireBoss_S26_2023_Idle_040}}
 {id=WavePlayer_2, properties {soundWaveAssetPtr=/VampireBoss/Audio/VO/Waves/VampireBoss_S26_2023_Idle_051}}
 {id=WavePlayer_3, properties {soundWaveAssetPtr=/VampireBoss/Audio/VO/Waves/VampireBoss_S26_2023_Idle_052}}
 {id=WavePlayer_4, properties {soundWaveAssetPtr=/VampireBoss/Audio/VO/Waves/VampireBoss_S26_2023_Idle_056}}
]
properties {
 attenuationSettings=/Game/Sounds/Fort_Human_VO_BR/NPCs/Shared/NPC_VO_Close_Att
 maxDistance=8500
 soundClassObject=&Fort_VO_Enemy_Critical
 volumeMultiplier=1.2
 concurrencySet [&/VampireBoss_Chatter_Conc]
}
