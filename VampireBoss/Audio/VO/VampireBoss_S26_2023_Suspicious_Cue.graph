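# Sound cue for the Vampire Boss's "Suspicious" voice lines, from the S26 2023 VO set.
# A root Random node picks among four equally weighted wave players, each holding one suspicious VO take.
# Playback uses the shared NPC alerted-VO attenuation, the critical enemy VO sound class, a slight volume boost and the boss's chatter concurrency group.
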SoundCue /VampireBoss/Audio/VO/VampireBoss_S26_2023_Suspicious_Cue
root=Random_0
nodes [
 {
  id=Random_0
  children [WavePlayer_0,WavePlayer_2,WavePlayer_3,WavePlayer_4]
  properties {bShouldExcludeFromBranchCulling=true, weights [1,1,1,1]}
 }
 {id=WavePlayer_0, properties {soundWaveAssetPtr=/VampireBoss/Audio/VO/Waves/VampireBoss_S26_2023_Suspicious_011}}
 {id=WavePlayer_2, properties {soundWaveAssetPtr=/VampireBoss/Audio/VO/Waves/VampireBoss_S26_2023_Suspicious_030}}
 {id=WavePlayer_3, properties {soundWaveAssetPtr=/VampireBoss/Audio/VO/Waves/VampireBoss_S26_2023_Suspicious_051}}
 {id=WavePlayer_4, properties {soundWaveAssetPtr=/VampireBoss/Audio/VO/Waves/VampireBoss_S26_2023_Suspicious_052}}
]
properties {
 attenuationSettings=/Game/Sounds/Fort_Human_VO_BR/NPCs/Shared/NPC_VO_Alerted_Att
 maxDistance=12500
 soundClassObject=&Fort_VO_Enemy_Critical
 volumeMultiplier=1.2
 concurrencySet [&/VampireBoss_Chatter_Conc]
}
